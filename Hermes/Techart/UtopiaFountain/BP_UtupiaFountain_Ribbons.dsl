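; Utopia fountain Blueprint that procedurally builds ribbon rings from spline meshes in its construction script.
; Each ring is tilted and scaled by index, with segments rolled progressively to twist the ribbon.
; SplineMeshMath supplies circle points, tangents and up vectors with alternating vertical wobble, and prints a debug value; SplineMath is unused.
; A spline component per ring reuses that math; BeginPlay, overlap and SetSpline are empty.
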(event EventBeginPlay)

(event Collision|EventActorBeginOverlap
  (params (OtherActor :type Actor%20Object%20Reference)))

(fn SetSpline)

(fn SplineMath
  (params (Index :type Integer) (Last%20Index :type Integer))
  (return (Math|Conversions|ToFloat(Integer) (Variables|GetIndex)) (+ (Variables|GetIndex) 1) (Math|Vector|RotateVector (Math|Vector|MakeVector (Variables|Default|GetRadius)) (Math|Rotator|MakeRotator 0.0 (* (/ 3.6e+02 (Variables|Default|GetNumSegments)) (Variables|GetIndex)))) (Math|Vector|RotateVector (Math|Vector|MakeVector (Variables|Default|GetRadius)) (Math|Rotator|MakeRotator 0.0 (* (/ 3.6e+02 (Variables|Default|GetNumSegments)) (select (> (+ (Variables|GetIndex) 1) (Variables|GetLast%20Index)) 0.0 (+ (Variables|GetIndex) 1))))) (* (Math|Vector|RotateVector (Math|Vector|MakeVector 0.0 0.0 1.0) (Math|Rotator|MakeRotator 0.0 (* (/ 3.6e+02 (Variables|Default|GetNumSegments)) (select (< (- (Variables|GetIndex) 1) 0) (Variables|GetLast%20Index) (- (Variables|GetIndex) 1))))) (* (Math|Trig|DegreesToRadians (/ 3.6e+02 (Variables|Default|GetNumSegments))) (Variables|Default|GetRadius))) (* (Math|Vector|RotateVector (Math|Vector|MakeVector 0.0 0.0 1.0) (Math|Rotator|MakeRotator 0.0 (* (/ 3.6e+02 (Variables|Default|GetNumSegments)) (select (> (+ (Variables|GetIndex) 1) (Variables|GetLast%20Index)) 0.0 (+ (Variables|GetIndex) 1))))) (* (Math|Trig|DegreesToRadians (/ 3.6e+02 (Variables|Default|GetNumSegments))) (Variables|Default|GetRadius))) (* (Math|Vector|RotateVector (Math|Vector|MakeVector 0.0 0.0 1.0) (Math|Rotator|MakeRotator 0.0 (* (/ 3.6e+02 (Variables|Default|GetNumSegments)) (Variables|GetIndex)))) (* (Math|Trig|DegreesToRadians (/ 3.6e+02 (Variables|Default|GetNumSegments))) (Variables|Default|GetRadius))) (* (Math|Vector|RotateVector (Math|Vector|MakeVector 0.0 0.0 1.0) (Math|Rotator|MakeRotator 0.0 (* (/ 3.6e+02 (Variables|Default|GetNumSegments)) (select (> (+ (select (> (+ (Variables|GetIndex) 1) (Variables|GetLast%20Index)) 0.0 (+ (Variables|GetIndex) 1)) 1) (Variables|GetLast%20Index)) 0.0 (+ (select (> (+ (Variables|GetIndex) 1) (Variables|GetLast%20Index)) 0.0 (+ (Variables|GetIndex) 1)) 1))))) (* (Math|Trig|DegreesToRadians (/ 3.6e+02 (Variables|Default|GetNumSegments))) (Variables|Default|GetRadius))) (Math|Vector|Normalize (+ (Math|Vector|RotateVector (Math|Vector|MakeVector (Variables|Default|GetRadius)) (Math|Rotator|MakeRotator 0.0 (* (/ 3.6e+02 (Variables|Default|GetNumSegments)) (select (> (+ (Variables|GetIndex) 1) (Variables|GetLast%20Index)) 0.0 (+ (Variables|GetIndex) 1))))) (/ (- (Math|Vector|RotateVector (Math|Vector|MakeVector (Variables|Default|GetRadius)) (Math|Rotator|MakeRotator 0.0 (* (/ 3.6e+02 (Variables|Default|GetNumSegments)) (Variables|GetIndex)))) (Math|Vector|RotateVector (Math|Vector|MakeVector (Variables|Default|GetRadius)) (Math|Rotator|MakeRotator 0.0 (* (/ 3.6e+02 (Variables|Default|GetNumSegments)) (select (> (+ (Variables|GetIndex) 1) (Variables|GetLast%20Index)) 0.0 (+ (Variables|GetIndex) 1)))))) 2.0))) (Math|Vector|Normalize (+ (Math|Vector|RotateVector (Math|Vector|MakeVector (Variables|Default|GetRadius)) (Math|Rotator|MakeRotator 0.0 (* (/ 3.6e+02 (Variables|Default|GetNumSegments)) (select (> (+ (select (> (+ (Variables|GetIndex) 1) (Variables|GetLast%20Index)) 0.0 (+ (Variables|GetIndex) 1)) 1) (Variables|GetLast%20Index)) 0.0 (+ (select (> (+ (Variables|GetIndex) 1) (Variables|GetLast%20Index)) 0.0 (+ (Variables|GetIndex) 1)) 1))))) (/ (- (Math|Vector|RotateVector (Math|Vector|MakeVector (Variables|Default|GetRadius)) (Math|Rotator|MakeRotator 0.0 (* (/ 3.6e+02 (Variables|Default|GetNumSegments)) (select (> (+ (Variables|GetIndex) 1) (Variables|GetLast%20Index)) 0.0 (+ (Variables|GetIndex) 1))))) (Math|Vector|RotateVector (Math|Vector|MakeVector (Variables|Default|GetRadius)) (Math|Rotator|MakeRotator 0.0 (* (/ 3.6e+02 (Variables|Default|GetNumSegments)) (select (> (+ (select (> (+ (Variables|GetIndex) 1) (Variables|GetLast%20Index)) 0.0 (+ (Variables|GetIndex) 1)) 1) (Variables|GetLast%20Index)) 0.0 (+ (select (> (+ (Variables|GetIndex) 1) (Variables|GetLast%20Index)) 0.0 (+ (Variables|GetIndex) 1)) 1)))))) 2.0)))))

(fn SplineMeshMath
  (params (Index :type Integer) (Sign :type Float%20%28double-precision%29 :default 1.0))
  (bind _returnvalue 4.0)
  (bind _returnvalue_2 (/ Index _returnvalue))
  (bind _returnvalue_3 (Math|Float|Fraction _returnvalue_2))
  (bind _returnvalue_4 (- _returnvalue_3 0.26))
  (bind _returnvalue_5 (Math|Float|Sign(Float) _returnvalue_4))
  (Utilities|FlowControl|Sequence
    (:then_0
      (Development|PrintString (Utilities|String|ToString(Float) _returnvalue_5)))
    (:then_1
      (return (Math|Vector|RotateVector (* (Math|Vector|Normalize (Math|Vector|MakeVector (Variables|Default|GetRadius) (Utilities|Operators|Multiply (Math|Float|Fraction (/ Index 2.0)) (* (Variables|Default|GetZOffsetRibbonWobble) Sign) _returnvalue_5))) (Variables|Default|GetRadius)) (Math|Rotator|MakeRotator 0.0 (* (/ 3.6e+02 (Variables|Default|GetNumSegments)) Index))) (Utilities|Operators|Multiply (Math|Vector|RotateVector (Math|Vector|MakeVector 0.0 0.0 1.0) (Math|Rotator|MakeRotator 0.0 (* (/ 3.6e+02 (Variables|Default|GetNumSegments)) Index))) (* (Math|Trig|DegreesToRadians (/ 3.6e+02 (Variables|Default|GetNumSegments))) (Variables|Default|GetRadius)) 1.0) (Math|Vector|RotateVector (* (Math|Vector|Normalize (Math|Vector|MakeVector (Variables|Default|GetRadius) (Utilities|Operators|Multiply (Math|Float|Fraction (/ (+ Index 1) 2.0)) (* (Variables|Default|GetZOffsetRibbonWobble) Sign) (Math|Float|Sign(Float) (- (Math|Float|Fraction (/ (+ Index 1) 4.0)) 0.26))))) (Variables|Default|GetRadius)) (Math|Rotator|MakeRotator 0.0 (* (/ 3.6e+02 (Variables|Default|GetNumSegments)) (+ Index 1)))) (Utilities|Operators|Multiply (Math|Vector|RotateVector (Math|Vector|MakeVector 0.0 0.0 1.0) (Math|Rotator|MakeRotator 0.0 (* (/ 3.6e+02 (Variables|Default|GetNumSegments)) (+ Index 1)))) (* (Math|Trig|DegreesToRadians (/ 3.6e+02 (Variables|Default|GetNumSegments))) (Variables|Default|GetRadius)) 1.0) (Math|Vector|Normalize (+ (Math|Vector|RotateVector (Math|Vector|MakeVector (Variables|Default|GetRadius)) (Math|Rotator|MakeRotator 0.0 (* (/ 3.6e+02 (Variables|Default|GetNumSegments)) (+ Index 1)))) (/ (- (Math|Vector|RotateVector (Math|Vector|MakeVector (Variables|Default|GetRadius)) (Math|Rotator|MakeRotator 0.0 (* (/ 3.6e+02 (Variables|Default|GetNumSegments)) Index))) (Math|Vector|RotateVector (Math|Vector|MakeVector (Variables|Default|GetRadius)) (Math|Rotator|MakeRotator 0.0 (* (/ 3.6e+02 (Variables|Default|GetNumSegments)) (+ Index 1))))) 2.0)))))))

(fn ConstructionScript
  (bind _returnvalue 1.0)
  (bind _returnvalue_2 1.8e+02)
  (bind _returnvalue_3 3.6e+02)
  (bind _returnvalue_4 2.0)
  (bind _numberofrings (Variables|Default|GetNumberOfRings))
  (bind _returnvalue_5 (/ _returnvalue_2 _numberofrings))
  (bind _numsegments (Variables|Default|GetNumSegments))
  (bind _number_of_twists (Variables|Default|GetNumber%20of%20Twists))
  (bind _returnvalue_6 (* _returnvalue_3 _number_of_twists))
  (bind _returnvalue_7 (/ _returnvalue_6 _numsegments))
  (bind _scale (Variables|Default|GetScale))
  (bind _ringscaleoffset (Variables|Default|GetRingScaleOffset))
  (for _index (range (Variables|Default|GetNumberOfRings))
    (Utilities|FlowControl|Sequence
      (:then_0
        (for _index_2 (range (Variables|Default|GetNumSegments))
          (bind _returnvalue_8 (AddComponent|NODE_AddSplineMeshComponent-0 :RelativeTransform (Math|Transform|MakeTransform "0, 0, 0" (Math|Rotator|MakeRotator (* _returnvalue_5 _index) (* _returnvalue_5 _index)) (Math|Vector|MakeVector (+ _returnvalue (* _ringscaleoffset _index)) (+ _returnvalue (* _ringscaleoffset _index)) (+ _returnvalue (* _ringscaleoffset _index))))))
          (SplineMesh|SetStartandEnd _returnvalue_8 (CallFunction|SplineMeshMath _index_2 (Math|Float|Sign(Float) (- (/ (+ _index 1) _returnvalue_4) 0.25))) (.Start%20Tangent (CallFunction|SplineMeshMath _index_2 (Math|Float|Sign(Float) (- (/ (+ _index 1) _returnvalue_4) 0.25)))) (.End%20Pos (CallFunction|SplineMeshMath _index_2 (Math|Float|Sign(Float) (- (/ (+ _index 1) _returnvalue_4) 0.25)))) (.End%20Tangent (CallFunction|SplineMeshMath _index_2 (Math|Float|Sign(Float) (- (/ (+ _index 1) _returnvalue_4) 0.25)))) true :!class /Script/Engine.SplineMeshComponent)
          (SplineMesh|SetSplineUpDir _returnvalue_8 (.In%20Spline%20Up%20Dir (CallFunction|SplineMeshMath _index_2 (Math|Float|Sign(Float) (- (/ (+ _index 1) _returnvalue_4) 0.25)))) true)
          (SplineMesh|SetStartRollDegrees _returnvalue_8 (* _returnvalue_7 _index_2) true)
          (SplineMesh|SetEndRollDegrees _returnvalue_8 (* _returnvalue_7 (+ _index_2 1)) true)
          (SplineMesh|SetEndScale _returnvalue_8 _scale true :!class /Script/Engine.SplineMeshComponent)
          (SplineMesh|SetStartScale _returnvalue_8 _scale true :!class /Script/Engine.SplineMeshComponent)))
      (:then_1
        (bind _returnvalue_9 (AddComponent|NODE_AddSplineComponent-3 :RelativeTransform (Math|Transform|MakeTransform "0, 0, 0" (Math|Rotator|MakeRotator (* _returnvalue_5 _index) (* _returnvalue_5 _index)) (Math|Vector|MakeVector (+ _returnvalue (* _ringscaleoffset _index)) (+ _returnvalue (* _ringscaleoffset _index)) (+ _returnvalue (* _ringscaleoffset _index))))))
        (Utilities|Array|Add (Variables|Default|GetSplineList))
        (for _index_3 (range (Variables|Default|GetNumSegments))
          (if (< _index_3 2)
            (Spline|AddSplinePointatIndex _returnvalue_9 (CallFunction|SplineMeshMath _index_3 (Math|Float|Sign(Float) (- (/ (+ _index 1) _returnvalue_4) 0.25))) _index_3 "Local" true)
            (else
              (Spline|SetLocationatSplinePoint _returnvalue_9 _index_3 (CallFunction|SplineMeshMath _index_3 (Math|Float|Sign(Float) (- (/ (+ _index 1) _returnvalue_4) 0.25))) "Local" true)))
          (Spline|SetTangentsatSplinePoint _returnvalue_9 _index_3 (.Start%20Tangent (CallFunction|SplineMeshMath _index_3 (Math|Float|Sign(Float) (- (/ (+ _index 1) _returnvalue_4) 0.25)))) (.Start%20Tangent (CallFunction|SplineMeshMath _index_3 (Math|Float|Sign(Float) (- (/ (+ _index 1) _returnvalue_4) 0.25)))) "Local" true)
          ; Create Niagara
          (Spline|SetUpVectoratSplinePoint _returnvalue_9 _index_3 (.In%20Spline%20Up%20Dir (CallFunction|SplineMeshMath _index_3 (Math|Float|Sign(Float) (- (/ (+ _index 1) _returnvalue_4) 0.25)))) "Local" true))))))
