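# Thin translucent, two-sided water ribbon material for the Utopia fountain FX, usable on Niagara sprites, mesh particles and spline meshes.
# A time-animated sine of particle position drives an offset combining UV-based width push and vertex-normal push.
# A UV sphere mask shapes refraction blend and surface coverage, and is also multiplied by particle alpha.
# A panned, flatness-adjusted normal map adds surface detail, with a tinted transmittance color feeding the thin translucent output.
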Material /Hermes/Techart/UtopiaFountain/FX/M_UtopiaFountain_WaterRibbon_FX
graph {
 properties {
  bAutomaticallySetUsageInEditor=false
  bUsedWithNiagaraMeshParticles=true
  bUsedWithNiagaraSprites=true
  bUsedWithSplineMeshes=true
  blendMode=BLEND_Translucent
  refractionMethod=RM_PixelNormalOffset
  shadingModel=MSM_ThinTranslucent
  translucencyLightingMode=TLM_SurfacePerPixelLighting
  twoSided=true
  usedShadingModels=Thin Translucent
 }
 nodes [
  n1 = TextureCoordinate()
  n2 = Multiply(A=Subtract(A=ComponentMask(None=@n1) {g=true}) {constB=0.5}, B=Transform(None=Constant3Vector() {constant=<rgba 0 1 0 0>}))
  n3 = ScalarParameter() {defaultValue=1, parameterName=WidthPush}
  n4 = ParticlePositionWS()
  n5 = ScalarParameter() {defaultValue=335.0579, parameterName=WPO_Movement_Frequency}
  n6 = Time()
  n7 = ScalarParameter() {defaultValue=0.2, parameterName=WPO_Movement_Speed}
  n8 = Sine(None=Add(A=Divide(A=@n4, B=@n5), B=Multiply(A=@n6, B=@n7)))
  n9 = MaterialFunctionCall(Float3=Add(A=@n8)) {materialFunction=&BreakOutFloat3Components}
  n10 = Add(A=Add(A=@n9.R, B=@n9.R), B=@n9.B)
  n11 = VertexNormalWS()
  Add(
   A=Multiply(A=Multiply(A=@n2, B=@n3), B=@n10)
   B=Multiply(A=Multiply(A=@n11, B=@n10), B=ScalarParameter() {defaultValue=1, parameterName=NormalPush})
  )
  ComponentMask(None=@n8) {r=true}
  n12 = SphereMask(
   A=TextureCoordinate()
   B=Constant2Vector() {g=0.5, r=0.5}
   Radius=ScalarParameter() {defaultValue=0.48, parameterName=Radius}
   Hardness=ScalarParameter() {parameterName=Hardness}
  )
  n13 = Power(Base=@n12, Exp=ScalarParameter() {parameterName=OpacityCoverage})
  LinearInterpolate(B=ScalarParameter() {defaultValue=1.0483, parameterName=Refraction}, Alpha=Saturate(None=@n13)) {constA=1}
  n14 = TextureCoordinate()
  n15 = AppendVector(A=Constant(), B=ScalarParameter() {parameterName=Pan Speed})
  MaterialFunctionCall(
   Normal=TextureSample(UVs=Panner(Coordinate=@n14, Speed=@n15)) {samplerType=SAMPLERTYPE_Normal, texture=&T_Sphere_N}.RGB
   Flatness=ScalarParameter() {defaultValue=-2}
  ) {materialFunction=&FlattenNormal}
  Multiply(A=@n12, B=ParticleColor().A)
  ParticleColor()
  ScalarParameter() {parameterName=Roughness}
  ThinTranslucentMaterialOutput!(TransmittanceColor=VectorParameter() {parameterName=TransmittanceColor, defaultValue=<rgba 0.6043 0.736 1 1>}.RGB, SurfaceCoverage=@n13)
  VectorParameter() {parameterName=TransmittanceColo2r, defaultValue=<rgba 0.146 0.2498 0.4323 1>}
  VectorParameter() {parameterName=Color, defaultValue=<rgba 0 0.2254 0.3542 1>}
  WorldPosition()
 ]
}
