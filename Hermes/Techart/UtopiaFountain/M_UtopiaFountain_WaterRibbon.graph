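# Thin-translucent, two-sided water ribbon material for the Utopia Fountain, enabled for spline and Niagara meshes.
# A panned mask drives depth-faded opacity, surface coverage and refraction, while a matching normal map adds detail and Fresnel blends two colours.
# World-position sine waves over time push vertices along normals and vary SplineThicken width.
# An optional static switch squashes the offset mesh toward its object position along a direction vector.
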Material /Hermes/Techart/UtopiaFountain/M_UtopiaFountain_WaterRibbon
graph {
 properties {
  bAutomaticallySetUsageInEditor=false
  bTangentSpaceNormal=false
  bUsedWithNiagaraMeshParticles=true
  bUsedWithSplineMeshes=true
  blendMode=BLEND_Translucent
  refractionMethod=RM_PixelNormalOffset
  shadingModel=MSM_ThinTranslucent
  translucencyLightingMode=TLM_SurfacePerPixelLighting
  twoSided=true
  usedShadingModels=Thin Translucent
 }
 nodes [
  n1 = VectorParameter() {parameterName=Tiling, defaultValue=<rgba 1 2 0 1>}
  n2 = Panner(
   Coordinate=Multiply(A=TextureCoordinate(), B=ComponentMask(None=@n1.RGB) {g=true, r=true})
   Speed=AppendVector(A=Constant(), B=ScalarParameter() {defaultValue=-0.25, parameterName=Pan Speed})
  )
  n3 = TextureSample(UVs=@n2) {samplerType=SAMPLERTYPE_Masks, texture=&/UtopiaFountain_Water_M}
  n4 = ScalarParameter() {defaultValue=48.8921, parameterName=DepthFade}
  DepthFade(Opacity=Multiply(A=@n3.R, B=ScalarParameter() {defaultValue=0.8694, parameterName=Opacity}), FadeDistance=@n4)
  LinearInterpolate(
   A=VectorParameter() {parameterName=Color, defaultValue=<rgba 0.0912 0.3318 0.3542 1>}.RGB
   B=VectorParameter() {parameterName=ColorB, defaultValue=<rgba 0.4948 0.4012 0.2912 1>}.RGB
   Alpha=Fresnel(ExponentIn=ScalarParameter() {defaultValue=0.8629, parameterName=Fresnell}, Normal=NamedRerouteUsage())
  )
  n5 = Power(Base=@n3.R, Exp=ScalarParameter() {defaultValue=0.28, parameterName=OpacityCoverage})
  NamedRerouteDeclaration(None=LinearInterpolate(B=ScalarParameter() {defaultValue=1.0483, parameterName=Refraction}, Alpha=Saturate(None=@n5)) {constA=1}) {name=Refraction, nodeColor=<rgba 0.9453 0 1 1>}
  n6 = WorldPosition()
  n7 = ScalarParameter() {defaultValue=895.3727, parameterName=WPO_Movement_Frequency}
  n8 = Time()
  n9 = ScalarParameter() {defaultValue=0.05, parameterName=WPO_Movement_Speed}
  n10 = Sine(None=Add(A=Divide(A=@n6.XYZ, B=@n7), B=Multiply(A=@n8, B=@n9)))
  n11 = MaterialFunctionCall(Float3=Divide(A=Add(A=@n10))) {materialFunction=&BreakOutFloat3Components}
  n12 = Add(A=Add(A=@n11.R, B=@n11.R), B=@n11.B)
  n13 = Multiply(A=VertexNormalWS(), B=@n12)
  n14 = ScalarParameter() {defaultValue=5, parameterName=NormalPush}
  n15 = Multiply(
   A=LinearInterpolate(
    A=ScalarParameter() {defaultValue=296.8581, parameterName=SplineThickenBaseWidth2}
    B=ScalarParameter() {defaultValue=216.7672, parameterName=SplineThickenBaseWidth}
    Alpha=@n12
   )
   B=ScalarParameter() {defaultValue=0.872, parameterName=SplineThickenBaseWidth_Overall}
  )
  n16 = MaterialFunctionCall(WidthBase=@n15, WidthTip=@n15) {materialFunction=&SplineThicken}
  NamedRerouteDeclaration(None=Add(A=Multiply(A=@n13, B=@n14), B=@n16.WorldPositionOffset)) {name=WPO, nodeColor=<rgba 0 1 0.625 1>}
  NamedRerouteDeclaration(None=MaterialFunctionCall(
    BaseNormal=NamedRerouteUsage()
    AdditionalNormal=TextureSample(UVs=@n2) {samplerType=SAMPLERTYPE_Normal, texture=&/UtopiaFountain_Water_N}.RGB
   ) {materialFunction=&BlendAngleCorrectedNormals}.Result) {name=Normal, nodeColor=<rgba 1 0.3047 0 1>}
  NamedRerouteDeclaration(None=@n16.Normal) {name=SplineThickenNormal, nodeColor=<rgba 0 0.0156 1 1>}
  NamedRerouteUsage()
  ScalarParameter() {parameterName=Roughness}
  n17 = NamedRerouteUsage()
  n18 = Normalize(VectorInput=VectorParameter() {parameterName=SequezeDiretion, defaultValue=<rgba 1 0 0 1>}.RGB)
  n19 = Subtract(A=Add(A=@n17, B=WorldPosition().XYZ), B=ObjectPositionWS())
  n20 = Multiply(
   A=Multiply(A=DotProduct(A=@n19, B=@n18)) {constB=-1}
   B=ScalarParameter() {defaultValue=1.1069, group=FrontendSqueze, parameterName=FrontendSqueze}
  )
  StaticSwitchParameter(True=Add(A=@n17, B=Multiply(A=@n18, B=@n20)), False=@n17) {group=FrontendSqueze, parameterName=Use Frontend Mesh Squeze}
  ThinTranslucentMaterialOutput!(
   TransmittanceColor=VectorParameter() {parameterName=TransmittanceColor, defaultValue=<rgba 1 1 1 1>}.RGB
   SurfaceCoverage=DepthFade(Opacity=@n5, FadeDistance=@n4)
  )
 ]
}
