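# Clear coat Nanite material for the Utopia fountain's orb statue, spinning under a layer of flowing water.
# Two panning samples in radial world-space coordinates, faded by height, form a water mask and normals that tint, refract and resurface two UV1-selected texture sets.
# World position offset rotates the orb on a wobbling axis with corrected normals, and can optionally squash it for the frontend.
# Emissive adds Fresnel skylight reflection and fake specular.
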Material /Hermes/Techart/UtopiaFountain/M_UtopiaRotatingWaterBall
graph {
 properties {
  bAutomaticallySetUsageInEditor=false
  bUseMaterialAttributes=true
  bUsedWithNanite=true
  bUsedWithNiagaraMeshParticles=true
  maxWorldPositionOffsetDisplacement=1300
  shadingModel=MSM_ClearCoat
  usedShadingModels=Clear Coat
 }
 nodes [
  Add(A=NamedRerouteUsage(), B=WorldPosition() {worldPositionShaderOffset=WPT_ExcludeAllShaderOffsets}.XYZ)
  Constant3Vector() {constant=<rgba 0 0 1 0>}
  Constant3Vector() {constant=<rgba 0 0 1 0>}
  n1 = ScalarParameter() {defaultValue=-0.0383, parameterName=WaterNormaLFlatten}
  n2 = LocalPosition()
  n3 = Divide(
   A=Subtract(A=@n2.Z, B=ScalarParameter() {defaultValue=1150.9244, parameterName=TopFadeOffset})
   B=ScalarParameter() {defaultValue=-48.022, parameterName=TopFade}
  )
  n4 = ScalarParameter() {defaultValue=590.2927, parameterName=BottomFadeOffset}
  n5 = Saturate(None=Divide(A=Subtract(A=@n2.Z, B=@n4), B=ScalarParameter() {defaultValue=-535.5261, parameterName=BottomFade}))
  n6 = WorldPosition()
  n7 = VertexInterpolator(VS=NamedRerouteUsage())
  n8 = Subtract(
   A=StaticSwitchParameter(True=Subtract(A=@n6.XYZ, B=@n7.PS), False=@n6.XYZ) {group=FrontendSqueze, parameterName=Use Frontend Mesh Squeze}
   B=ObjectPositionWS()
  )
  n9 = Constant3Vector() {constant=<rgba 1 1 0 0>}
  n10 = MaterialFunctionCall(Vector or UVs=Normalize(VectorInput=Multiply(A=@n8, B=@n9))) {materialFunction=&VectorToRadialValue}
  n11 = AppendVector(
   A=Multiply(A=ComponentMask(None=@n10.Radial Coordinates) {r=true}, B=ScalarParameter() {defaultValue=4, parameterName=HorizontalTiling})
   B=Divide(A=ComponentMask(None=@n6.XYZ) {b=true}, B=ScalarParameter() {defaultValue=-723.0895, parameterName=HeightTiling})
  )
  n12 = ScalarParameter() {defaultValue=-0.001, parameterName=PanSpeed_Water_X}
  n13 = ScalarParameter() {defaultValue=-0.3, parameterName=PanSpeed_Water}
  n14 = Panner(Coordinate=@n11, Speed=AppendVector(A=@n12, B=@n13))
  n15 = TextureCoordinate()
  n16 = DDX(Value=@n15)
  n17 = DDY(Value=@n15)
  n18 = Panner(
   Coordinate=Add(A=@n11, B=Constant2Vector() {g=0.25, r=0.5})
   Speed=AppendVector(A=Multiply(A=@n12) {constB=0.25}, B=Multiply(A=@n13) {constB=0.9})
  )
  n19 = Power(
   Base=Max(
    A=TextureSample!(UVs=@n14, Tex, "DDX(UVs)"=@n16, "DDY(UVs)"=@n17, Apply View MipBias) {mipValueMode=TMVM_Derivative, texture=&/UtopiaFountain_Water_Orb_M}.R
    B=TextureSample(UVs=@n18, "DDX(UVs)"=@n16, "DDY(UVs)"=@n17) {mipValueMode=TMVM_Derivative, texture=&UtopiaFountain_Water_Orb_M}.R
   )
   Exp=ScalarParameter() {defaultValue=1.5827, parameterName=WaterContrast}
  )
  n20 = ScalarParameter() {defaultValue=0.53, parameterName=WaterOpcaity}
  n21 = Multiply(A=Multiply(A=Saturate(None=@n3), B=OneMinus(None=@n5)), B=Saturate(None=Multiply(A=@n19, B=@n20)))
  LinearInterpolate(A=@n1, Alpha=NamedRerouteDeclaration(None=Saturate(None=@n21)) {name=WaterMask, nodeColor=<rgba 1 0 0.6797 1>})
  n22 = CameraVectorWS()
  Multiply(A=@n22) {constB=-1}
  n23 = VertexNormalWS()
  n24 = TextureSample(UVs=@n14, "DDX(UVs)"=@n16, "DDY(UVs)"=@n17) {mipValueMode=TMVM_Derivative, samplerType=SAMPLERTYPE_Normal, texture=&/UtopiaFountain_Water_Orb_N}
  n25 = TextureSample(UVs=@n18, "DDX(UVs)"=@n16, "DDY(UVs)"=@n17) {mipValueMode=TMVM_Derivative, samplerType=SAMPLERTYPE_Normal, texture=&UtopiaFountain_Water_Orb_N}
  n26 = Multiply(
   A=MaterialFunctionCall(
    Normal=MaterialFunctionCall(BaseNormal=@n24.RGB, AdditionalNormal=@n25.RGB) {materialFunction=&BlendAngleCorrectedNormals}.Result
    Flatness=@n1
   ) {materialFunction=&FlattenNormal}.Result
   B=VectorParameter() {parameterName=NormalCorrection, defaultValue=<rgba 1 1 1 1>}.RGB
  )
  NamedRerouteDeclaration(None=Transform(None=MaterialFunctionCall(BaseNormal=@n23, AdditionalNormal=@n26) {materialFunction=&BlendAngleCorrectedNormals}.Result) {transformSourceType=TRANSFORMSOURCE_World, transformType=TRANSFORM_Tangent}) {name=WaterNormal, nodeColor=<rgba 1 0 0.6797 1>}
  n27 = Multiply(A=Time(), B=ScalarParameter() {defaultValue=0.01, parameterName=Rotation_Speed_2})
  n28 = Sine(None=@n27)
  n29 = Sine(None=Add(A=@n27) {constB=0.5})
  n30 = Normalize(VectorInput=AppendVector(A=AppendVector(A=@n28, B=@n29), B=Constant() {r=1}))
  n31 = Multiply(A=Time(), B=ScalarParameter() {defaultValue=0.02, parameterName=Rotation_Speed})
  n32 = TransformPosition(None=Constant3Vector() {constant=<rgba 0 0 612 0>}) {transformSourceType=TRANSFORMPOSSOURCE_Instance, transformType=TRANSFORMPOSSOURCE_World}
  n33 = WorldPosition()
  NamedRerouteDeclaration(None=StaticSwitchParameter(True=RotateAboutAxis(NormalizedRotationAxis=@n30, RotationAngle=@n31, PivotPoint=@n32, Position=@n33.XYZ), False=Constant3Vector()) {defaultValue=true, parameterName=UseWPORotation}) {name=WPO, nodeColor=<rgba 0 1 0.6016 1>}
  n34 = Normalize(VectorInput=VectorParameter() {parameterName=SequezeDiretion, defaultValue=<rgba 1 0 0 1>}.RGB)
  n35 = NamedRerouteUsage()
  n36 = WorldPosition() {worldPositionShaderOffset=WPT_ExcludeAllShaderOffsets}
  n37 = DotProduct(A=Subtract(A=Add(A=@n35, B=@n36.XYZ), B=ObjectPositionWS()), B=@n34)
  n38 = Multiply(
   A=@n34
   B=Multiply(A=Multiply(A=@n37) {constB=-1}, B=ScalarParameter() {defaultValue=0.99, group=FrontendSqueze, parameterName=FrontendSqueze})
  )
  NamedRerouteDeclaration(None=@n38) {name=Squeze, nodeColor=<rgba 1 0 0.7969 1>}
  NamedRerouteUsage()
  ObjectPositionWS()
  n39 = NamedRerouteUsage()
  n40 = Multiply(A=NamedRerouteUsage(), B=ScalarParameter() {defaultValue=0.0592, parameterName=Refraction})
  n41 = Add(A=TextureCoordinate(), B=ComponentMask(None=Multiply(A=@n39, B=@n40)) {g=true, r=true})
  n42 = TextureSampleParameter2D(UVs=@n41) {parameterName=Diffuse, texture=/Hermes/Props/Utopia/Utopia_OrbStatue_A/Textures/Nanite/T_NITE_Utopia_OrbStatue_A_A_D}
  n43 = TextureCoordinate() {coordinateIndex=1}
  n44 = Saturate(None=Floor(None=ComponentMask(None=@n43) {r=true}))
  n45 = StaticSwitchParameter(
   True=LinearInterpolate(
    A=@n42.RGB
    B=TextureSampleParameter2D(UVs=@n41) {parameterName=Diffuse2, texture=/Hermes/Props/Utopia/Utopia_OrbStatue_A/Textures/Nanite/T_NITE_Utopia_OrbStatue_A_B_D}.RGB
    Alpha=@n44
   )
   False=@n42.RGB
  ) {defaultValue=true, parameterName=UseLayer2}
  n46 = VectorParameter() {parameterName=WaterColor, defaultValue=<rgba 2 1.099 0.3125 1>}
  n47 = ScalarParameter() {defaultValue=0.649, parameterName=WaterOpacityColor}
  n48 = ScalarParameter() {defaultValue=0.1908, parameterName=ConstantBlend}
  n49 = NamedRerouteUsage()
  n50 = Saturate(None=Multiply(A=Multiply(A=@n48, B=@n49), B=ScalarParameter() {defaultValue=10.975, parameterName=WaterOpacity}))
  n51 = TextureSampleParameter2D(UVs=@n41) {
   parameterName=SMR
   samplerType=SAMPLERTYPE_Masks
   texture=/Hermes/Props/Utopia/Utopia_OrbStatue_A/Textures/Nanite/T_NITE_Utopia_OrbStatue_A_A_S
  }
  n52 = TextureSampleParameter2D(UVs=@n41) {
   parameterName=SMR2
   samplerType=SAMPLERTYPE_Masks
   texture=/Hermes/Props/Utopia/Utopia_OrbStatue_A/Textures/Nanite/T_NITE_Utopia_OrbStatue_A_B_S
  }
  n53 = StaticSwitchParameter(True=LinearInterpolate(A=@n51.RGB, B=@n52.RGB, Alpha=@n44), False=@n51.RGB) {defaultValue=true, parameterName=UseLayer2}
  n54 = TextureSampleParameter2D(UVs=@n41) {
   parameterName=Normals
   samplerType=SAMPLERTYPE_Normal
   texture=/Hermes/Props/Utopia/Utopia_OrbStatue_A/Textures/Nanite/T_NITE_Utopia_OrbStatue_A_A_N
  }
  n55 = TextureSampleParameter2D(UVs=@n41) {
   parameterName=Normals2
   samplerType=SAMPLERTYPE_Normal
   texture=/Hermes/Props/Utopia/Utopia_OrbStatue_A/Textures/Nanite/T_NITE_Utopia_OrbStatue_A_B_N
  }
  n56 = StaticSwitchParameter(True=LinearInterpolate(A=@n54.RGB, B=@n55.RGB, Alpha=@n44), False=@n54.RGB) {defaultValue=true, parameterName=UseLayer2}
  n57 = LinearInterpolate(
   A=@n56
   B=MaterialFunctionCall(BaseNormal=@n56, AdditionalNormal=NamedRerouteUsage()) {materialFunction=&BlendAngleCorrectedNormals}.Result
   Alpha=@n50
  )
  n58 = StaticSwitchParameter(
   True=MaterialFunctionCall(
    BaseNormal=MaterialFunctionCall(Rotation Axis=@n30, Rotation Angle=@n31) {materialFunction=&FixRotateAboutAxisNormals}.New Tangent Space Vertex Normal
    AdditionalNormal=@n57
   ) {materialFunction=&BlendAngleCorrectedNormals}.Result
   False=@n57
  ) {defaultValue=true, parameterName=UseWPORotation}
  n59 = Multiply(A=ScalarParameter() {defaultValue=0.4, parameterName=EnvSampleBirghtness}, B=SkyLightEnvMapSample(Direction=ReflectionVectorWS()))
  n60 = Transform(None=@n58)
  n61 = Fresnel(ExponentIn=ScalarParameter() {defaultValue=4.5034, parameterName=Fresnel}, Normal=@n60) {baseReflectFraction=0}
  n62 = ScalarParameter() {defaultValue=0.5, parameterName=FakeSpecBrightness}
  n63 = Power(Base=DotProduct(A=@n60, B=@n22), Exp=ScalarParameter() {defaultValue=71.052, parameterName=Fresnel2})
  SetMaterialAttributes!(
   MaterialAttributes
   Base Color=LinearInterpolate(A=@n45, B=LinearInterpolate(A=@n45, B=@n46.RGB, Alpha=@n47), Alpha=@n50)
   Metallic=LinearInterpolate(A=Constant() {r=1}, B=ScalarParameter() {defaultValue=0.6283, parameterName=MetallicWater}, Alpha=@n50)
   Specular=Constant() {r=1}
   Roughness=LinearInterpolate(A=ComponentMask(None=@n53) {b=true}, B=ScalarParameter() {parameterName=RoughnessWater}, Alpha=@n50)
   Normal=@n58
   World Position Offset=StaticSwitchParameter(True=Add(A=@n35, B=@n38), False=@n35) {group=FrontendSqueze, parameterName=Use Frontend Mesh Squeze}
   Clear Coat=LinearInterpolate(
    A=ScalarParameter() {defaultValue=1, parameterName=ClearCoat_A}
    B=ScalarParameter() {defaultValue=1, parameterName=ClearCoat_B}
    Alpha=NamedRerouteUsage()
   )
   Clear Coat Roughness=LinearInterpolate(
    A=ScalarParameter() {defaultValue=0.382, parameterName=ClearCoat_Rough_A}
    B=ScalarParameter() {parameterName=ClearCoat_Rough_B}
    Alpha=NamedRerouteUsage()
   )
   Emissive Color=Add(A=Multiply(A=@n59, B=@n61), B=Multiply(A=@n62, B=@n63))
  ) {attributeSetTypes [
    #1
    #2
    #3
    #4
    #5
    #6
    #7
    #8
    #9
   ]}
  Transform()
  VertexNormalWS()
 ]
}
